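# VFX library material function that outputs a mask fading pixels out by how far they sit behind opaque scene geometry.
# An If node returns 1 where the pixel is at or in front of scene depth.
# Behind it, the pixel-minus-scene depth gap is offset by a bias, divided, clamped, inverted and raised to an exponent.
# Bias, divisor and exponent are inputs with parameter defaults, so the negative default bias delays the fade.
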MaterialFunction /Game/Effects/Fort_Effects/Materials/_Functions/MF_DepthFadeBehindOpaque_01
graph {
 properties {bExposeToLibrary=true, libraryCategoriesText [VFX]}
 nodes [
  n1 = Constant() {r=1}
  n2 = Add(
   A=Subtract(A=PixelDepth().R, B=SceneDepth().R)
   B=FunctionInput(Preview=ScalarParameter() {defaultValue=-400, parameterName=Depth Fade Behind Bias}) {bUsePreviewValueAsDefault=true, inputName=Depth Fade Behind Bias, inputType=FunctionInput_Scalar}
  )
  n3 = FunctionInput(Preview=ScalarParameter() {defaultValue=100, parameterName=Depth Fade Behind Divisor}) {bUsePreviewValueAsDefault=true, inputName=Depth Fade Behind Divisor, inputType=FunctionInput_Scalar, sortPriority=1}
  n4 = OneMinus(None=Clamp(None=Divide(A=@n2, B=@n3)))
  n5 = FunctionInput(Preview=ScalarParameter() {defaultValue=6, parameterName=Depth Fade Behind Exponent}) {bUsePreviewValueAsDefault=true, inputName=Depth Fade Behind Exponent, inputType=FunctionInput_Scalar, sortPriority=2}
  FunctionOutput(None=If(A=SceneDepth().R, B=PixelDepth().R, "A > B"=@n1, "A == B"=@n1, "A < B"=Power(Base=@n4, Exp=@n5)))
 ]
}
