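# Unlit modulate particle material for the weak point effect background, tinting what lies behind it with a soft circle of particle color.
# A UV-centered SphereMask with Radius and Hardness parameters blends from neutral white to particle color.
# With depth test disabled, MF_DepthFadeBehindOpaque_01 controls fading back to white relative to opaque geometry.
# A quality switch replaces that depth fade with a constant at Low quality.
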Material /Game/Effects/Fort_Effects/BluePrints/WeakPointEffect/M_Weakpoint_BG_01
graph {
 properties {
  bDisableDepthTest=true
  bEnableResponsiveAA=true
  bUsedWithNiagaraSprites=true
  bUsedWithParticleSprites=true
  blendMode=BLEND_Modulate
  pixelDepthOffsetMode=PDOM_Legacy
  shadingModel=MSM_Unlit
  translucencyPass=MTP_BeforeDOF
  usedShadingModels=Unlit
 }
 nodes [
  n1 = TextureCoordinate()
  n2 = Constant() {r=0.5}
  n3 = ScalarParameter() {defaultValue=0.5, parameterName=Radius}
  n4 = ScalarParameter() {defaultValue=0.4, parameterName=Hardness}
  n5 = MaterialFunctionCall!(Depth Fade Behind Bias, Depth Fade Behind Divisor, Depth Fade Behind Exponent) {materialFunction=/Game/Effects/Fort_Effects/Materials/_Functions/MF_DepthFadeBehindOpaque_01}
  LinearInterpolate(
   B=LinearInterpolate(B=ParticleColor().RGB, Alpha=SphereMask(A=@n1, B=@n2, Radius=@n3, Hardness=@n4)) {constA=1}
   Alpha=QualitySwitch(Default=@n5.Result, Low=Constant() {r=1}, High=@n5.Result, Medium=@n5.Result)
  ) {constA=1}
 ]
}
