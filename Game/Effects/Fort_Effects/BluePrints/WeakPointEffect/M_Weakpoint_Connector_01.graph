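# Unlit translucent material for the weak point connector effect, drawing a stepped trail of dots and an arrow, separated by vertex red.
# A timestamp function, or looping debug time, drives a sweep through UV-stepped dots while fading the arrow.
# Fresnel adds to both color blends, and a local X offset stretches the dots along V by Scale.
# Depth testing is disabled, with a depth-fade-behind-opaque function modulating opacity.
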Material /Game/Effects/Fort_Effects/BluePrints/WeakPointEffect/M_Weakpoint_Connector_01
graph {
 properties {
  bDisableDepthTest=true
  bEnableResponsiveAA=true
  blendMode=BLEND_Translucent
  pixelDepthOffsetMode=PDOM_Legacy
  shadingModel=MSM_Unlit
  translucencyLightingMode=TLM_VolumetricPerVertexDirectional
  usedShadingModels=Unlit
 }
 nodes [
  n1 = ScalarParameter() {defaultValue=7, parameterName=Steps}
  n2 = Multiply(A=ComponentMask(None=TextureCoordinate()) {g=true}, B=@n1)
  n3 = OneMinus(None=Divide(A=Ceil(None=@n2), B=@n1))
  n4 = Time() {period=1}
  n5 = ScalarParameter() {defaultValue=0.75, parameterName=Speed A}
  n6 = StaticSwitchParameter(
   True=Frac(None=Multiply(A=@n4, B=@n5))
   False=MaterialFunctionCall(
    TimestampA=ScalarParameter() {parameterName=TimestampA}
    TransitionLength_In=ScalarParameter() {defaultValue=0.5, parameterName=TransitionLength_In}
   ) {materialFunction=&MF_Timestamp_InOut_01}.Linear
  ) {parameterName=DebugTime}
  n7 = Multiply(A=@n6) {constB=2}
  n8 = Constant() {r=1}
  n9 = ScalarParameter() {defaultValue=7, parameterName=Steps}
  n10 = Multiply(A=@n7, B=@n9)
  n11 = OneMinus(None=Divide(A=Ceil(None=@n10), B=@n9))
  n12 = Multiply(A=If(A=@n3, B=@n7, "A > B"=Constant(), "A == B"=@n8, "A < B"=@n8), B=Clamp(None=Add(A=@n11, B=@n3)))
  n13 = Multiply(
   A=Power(Base=@n12, Exp=ScalarParameter() {defaultValue=25, parameterName=Sweep Exponent}) {constExponent=20}
   B=ScalarParameter() {defaultValue=8, parameterName=Sweep Brightness}
  )
  n14 = MaterialFunctionCall(
   Brightness=ScalarParameter() {defaultValue=1.1, parameterName=Fresnel Brightness}
   Exponent=ScalarParameter() {defaultValue=5, parameterName=Fresnel Exponent}
  ) {materialFunction=&MF_Fresnel_FX}
  LinearInterpolate(
   A=LinearInterpolate(
    A=VectorParameter() {parameterName=Dots Outer, defaultValue=<rgba 0 0.06 0.3 1>}.RGB
    B=VectorParameter() {parameterName=Dots Inner, defaultValue=<rgba 1.14 3.312 12 1>}.RGB
    Alpha=Add(A=@n13, B=@n14.Fresnel)
   )
   B=LinearInterpolate(
    A=VectorParameter() {parameterName=Arrow Outer, defaultValue=<rgba 2.5 0.2998 0.2625 0>}.RGB
    B=VectorParameter() {parameterName=Arrow Inner, defaultValue=<rgba 8 0.692 0.44 0>}.RGB
    Alpha=@n14.Fresnel
   )
   Alpha=VertexColor().R
  ) {constB=0}
  n15 = TextureCoordinate()
  n16 = MaterialFunctionCall(X=OneMinus(None=ComponentMask(None=@n15) {g=true})) {materialFunction=&MakeFloat3}
  n17 = ScalarParameter() {defaultValue=100, parameterName=Scale}
  LinearInterpolate(A=Transform(None=Multiply(A=@n16.Result, B=@n17)) {transformSourceType=TRANSFORMSOURCE_Local}, Alpha=VertexColor().R) {constB=0}
  n18 = Frac(None=@n2)
  n19 = Multiply(A=OneMinus(None=@n18), B=@n18)
  n20 = ScalarParameter() {defaultValue=-0.04, parameterName=Line Bias}
  n21 = Clamp(None=Multiply(A=Add(A=@n19, B=@n20), B=ScalarParameter() {defaultValue=8.5, parameterName=Line Scale}))
  n22 = OneMinus(None=Multiply(A=@n6) {constB=2.5})
  Multiply(
   A=LinearInterpolate(A=Multiply(A=@n21, B=@n12), B=Clamp(None=@n22), Alpha=VertexColor().R)
   B=MaterialFunctionCall!(Depth Fade Behind Bias, Depth Fade Behind Divisor, Depth Fade Behind Exponent) {materialFunction=/Game/Effects/Fort_Effects/Materials/_Functions/MF_DepthFadeBehindOpaque_01}.Result
  )
 ]
}
