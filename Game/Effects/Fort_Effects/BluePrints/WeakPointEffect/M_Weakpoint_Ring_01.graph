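# Unlit translucent material for the weak point ring effect on mesh particles, drawn as a scanlined hologram without depth testing.
# A hologram wiggle drives vertex offset and scanline blocks, while dynamic parameters control wiggle, opacity and collapse toward the actor position.
# Fresnel times scanlines blends two charged colors, tinted by particle color.
# Opacity combines particle alpha, scanlines and a behind-opaque depth fade; Low quality skips fresnel and that fade.
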Material /Game/Effects/Fort_Effects/BluePrints/WeakPointEffect/M_Weakpoint_Ring_01
graph {
 properties {
  bDisableDepthTest=true
  bEnableResponsiveAA=true
  bUsedWithMeshParticles=true
  bUsedWithNiagaraMeshParticles=true
  blendMode=BLEND_Translucent
  pixelDepthOffsetMode=PDOM_Legacy
  shadingModel=MSM_Unlit
  usedShadingModels=Unlit
 }
 nodes [
  n1 = MaterialFunctionCall!(
   WPO Intensity A=VectorParameter() {parameterName=WPO Intensity A, defaultValue=<rgba 0.5022 0.4523 0.558 0>}.RGB
   WPO Intensity B=VectorParameter() {parameterName=WPO Intensity B, defaultValue=<rgba 1 5 1 0>}.RGB
   Time
   Bias=DynamicParameter() {defaultValue=<rgba 0 1 1 1>, paramNames [Wiggle Bias, OpacityHack, Scale Hack, Param4]}.Wiggle Bias
  ) {materialFunction=/Game/Effects/Fort_Effects/Effects/Atlas/Materials/MF_Atlas_HologramWiggle_01}
  LinearInterpolate(
   A=Subtract(A=ActorPositionWS(), B=WorldPosition().XYZ)
   B=@n1.Result
   Alpha=DynamicParameter() {defaultValue=<rgba 0 1 1 1>, paramNames [Wiggle Bias, OpacityHack, Scale Hack, Param4]}.Scale Hack
  )
  n2 = VectorParameter() {parameterName=Color Inner Charged, defaultValue=<rgba 0.6552 1.0222 8 1>}
  n3 = VectorParameter() {parameterName=Color Outer Charged, defaultValue=<rgba 0.105 0.6362 3 1>}
  n4 = MaterialFunctionCall(
   Brightness=ScalarParameter() {defaultValue=1.7, parameterName=Fresnel Brightness}
   Exponent=ScalarParameter() {defaultValue=4, parameterName=Fresnel Exponent}
  ) {materialFunction=&MF_Fresnel_FX}
  n5 = MaterialFunctionCall!(
   Number of Scanlines=ScalarParameter() {defaultValue=200, parameterName=Scanline Amount}
   Scanline Speed
   "Scanlines Thickness (Bias)"=ScalarParameter() {defaultValue=-0.15, parameterName=ScanlineSize}
   Scanlines Dark Value=ScalarParameter() {defaultValue=0.2, parameterName=Scanlines Dark}
   Scanlines Light Value=ScalarParameter() {defaultValue=1, parameterName=Scanlines Light}
   Block Opacity=@n1.Intense 0-1
   UVs
   Time
   Block Repeat=ScalarParameter() {defaultValue=6, parameterName=Block Repeat}
  ) {materialFunction=/Game/Effects/Fort_Effects/Effects/Atlas/Materials/MF_Atlas_Scanlines_01}
  n6 = Multiply(A=QualitySwitch(Default=@n4.Fresnel, Low=Constant() {r=1}, High=@n4.Fresnel, Medium=@n4.Fresnel), B=@n5.Result) {constA=1}
  n7 = ParticleColor()
  MaterialFunctionCall(Emissive=Multiply(A=LinearInterpolate(A=@n2.RGB, B=@n3.RGB, Alpha=@n6), B=@n7.RGB)) {materialFunction=&MF_FX_EmissiveAdjustments}
  n8 = MaterialFunctionCall!(Depth Fade Behind Bias, Depth Fade Behind Divisor, Depth Fade Behind Exponent) {materialFunction=/Game/Effects/Fort_Effects/Materials/_Functions/MF_DepthFadeBehindOpaque_01}
  Multiply(
   A=Multiply(
    A=Multiply(A=@n7.A, B=@n5.Result)
    B=DynamicParameter() {defaultValue=<rgba 0 1 1 1>, paramNames [Wiggle Bias, OpacityHack, Scale Hack, Param4]}.OpacityHack
   )
   B=QualitySwitch(Default=@n8.Result, Low=Constant() {r=1}, High=@n8.Result, Medium=@n8.Result)
  )
 ]
}
