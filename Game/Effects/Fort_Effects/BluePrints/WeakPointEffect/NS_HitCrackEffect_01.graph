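# NS_HitCrackEffect_01 is a Niagara system in the weak point effect folder, apparently the visual for a crack when a weak point is hit.
# It defines five emitters named Ring_Base, Core, Core_Long, GlowBits and Flash, suggesting a layered ring, core, glow fragment and flash burst.
# It exposes no user variables and lists SystemState as its only module, with no renderers, data interfaces or dynamic inputs recorded.
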NiagaraSystem /Game/Effects/Fort_Effects/BluePrints/WeakPointEffect/NS_HitCrackEffect_01
dependencies {usedDataInterfaces [], usedDynamicInputs [], usedModules [&SystemState]}
emitters {
 Core {rendererClasses [], renderers []}
 Core_Long {rendererClasses [], renderers []}
 Flash {rendererClasses [], renderers []}
 GlowBits {rendererClasses [], renderers []}
 Ring_Base {rendererClasses [], renderers []}
}
summary {
 systemName=NS_HitCrackEffect_01
 emitters [
  {emitterName=Ring_Base, rendererClasses []}
  {emitterName=Core, rendererClasses []}
  {emitterName=Core_Long, rendererClasses []}
  {emitterName=GlowBits, rendererClasses []}
  {emitterName=Flash, rendererClasses []}
 ]
 userVariables []
}
