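# Niagara system for the weak point effect, which by its name and emitters forms a targeting marker.
# It is built from six emitters: a background, base ring, outer ring, starter rings, dot and crosshair.
# User parameters expose a vector size scale and a 2D sprite size scale, both defaulting to one.
# The system lists only the SystemState module, and no renderers are recorded for any emitter.
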NiagaraSystem /Game/Effects/Fort_Effects/BluePrints/WeakPointEffect/NS_WeakPointEffect_01
dependencies {usedDataInterfaces [], usedDynamicInputs [], usedModules [&SystemState]}
emitters {
 BG {rendererClasses [], renderers []}
 Crosshair {rendererClasses [], renderers []}
 Dot {rendererClasses [], renderers []}
 Outer_Ring {rendererClasses [], renderers []}
 Ring_Base {rendererClasses [], renderers []}
 StarterRings {rendererClasses [], renderers []}
}
summary {
 systemName=NS_WeakPointEffect_01
 emitters [
  {emitterName=Ring_Base, rendererClasses []}
  {emitterName=Dot, rendererClasses []}
  {emitterName=Outer_Ring, rendererClasses []}
  {emitterName=StarterRings, rendererClasses []}
  {emitterName=Crosshair, rendererClasses []}
  {emitterName=BG, rendererClasses []}
 ]
 userVariables [
  User.SizeScale {type=Vector3f, description="", defaultValue {struct=Vector3f, value=<1 1 1>}}
  User.SpriteSizeScale {type=Vector2f, description="", defaultValue {struct=Vector2f, value=<1 1>}}
 ]
}
