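; WeakSpot is a Blueprint for weak point visual effects that supports both Niagara and Cascade particles.
; SetFXTemplate prefers a valid Niagara system over Cascade, retargeting the current FX component or replacing it with a root-attached one when the type differs.
; SetSpriteScale stores the scale and applies it through Niagara's SpriteSizeScale variable or as Cascade component world scale.
; The construction script caches a dynamic material from the ConnectTheDots component.
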(fn SetFXTemplate
  (params (Niagara :type Niagara%20System%20Object%20Reference) (Cascade :type Particle%20System%20Object%20Reference))
  (bind _fxcomponent (Variables|Default|GetFXComponent))
  (bind _fxcomponent_2 (Variables|Default|GetFXComponent))
  (if (Utilities|IsValid Niagara :!class /Script/Engine.KismetSystemLibrary)
    (bind _asniagara_particle_system_component (Utilities|Casting|CastToNiagaraComponent (Variables|Default|GetFXComponent))
      (:then
        (Niagara|SetNiagaraSystemAsset _asniagara_particle_system_component Niagara true))
      (:CastFailed
        (if (Utilities|IsValid _fxcomponent :!class /Script/Engine.KismetSystemLibrary)
          (Components|DestroyComponent _fxcomponent))
        (bind _returnvalue (Niagara|SpawnSystemAttached Niagara (Variables|Transformation|GetRootComponent) "None" "0, 0, 0" "0, 0, 0" "KeepRelativeOffset" false false))
        (Variables|Default|SetFXComponent _returnvalue)))
    (else
      (bind _ascascade_particle_system_component (Utilities|Casting|CastToParticleSystemComponent (Variables|Default|GetFXComponent))
        (:then
          (Effects|Components|ParticleSystem|SetTemplate _ascascade_particle_system_component Cascade))
        (:CastFailed
          (if (Utilities|IsValid _fxcomponent_2 :!class /Script/Engine.KismetSystemLibrary)
            (Components|DestroyComponent _fxcomponent_2))
          (bind _returnvalue_2 (Effects|SpawnEmitterAttached :EmitterTemplate Cascade :AttachToComponent (Variables|Transformation|GetRootComponent) :AttachPointName "None" :Scale "1.000000,1.000000,1.000000" :LocationType "KeepRelativeOffset" :bAutoDestroy false :PoolingMethod "None" :bAutoActivate false))
          (Variables|Default|SetFXComponent _returnvalue_2))))))

(fn SetSpriteScale
  (params (SpriteScale :type Float%20%28double-precision%29))
  (bind _fxcomponent (Variables|Default|GetFXComponent))
  (bind _output_get (Variables|Default|SetFXSpriteScale SpriteScale))
  (bind _asniagara_particle_system_component (Utilities|Casting|CastToNiagaraComponent _fxcomponent)
    (:then
      (Niagara|SetNiagaraVariable(Vector2) _asniagara_particle_system_component "SpriteSizeScale" (Math|Conversions|ToVector2D _output_get))
      (Transformation|SetWorldScale3D _asniagara_particle_system_component "1.000000,1.000000,1.000000"))
    (:CastFailed
      (bind _ascascade_particle_system_component (Utilities|Casting|CastToParticleSystemComponent _fxcomponent))
      (Transformation|SetWorldScale3D _ascascade_particle_system_component (Math|Conversions|ToVector(Float) _output_get)))))

(fn ConstructionScript
  (bind _returnvalue (Rendering|Material|CreateDynamicMaterialInstance (Variables|Default|GetConnectTheDotsComponent) 0 0 "None" :!class /Script/Engine.PrimitiveComponent))
  (Variables|Default|SetConnectMID _returnvalue))
