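# BlendMorphs is a material function in the WeakPointEffect folder that turns one scalar input into a two-stage static mesh morph offset.
# It reads both World Position Offset outputs of StaticMeshMorphTargets.
# Doubling and clamping the input makes its first half scale in morph target 1.
# Subtracting and clamping drives a lerp over the second half from target 1 to target 2, chaining both morphs on one control.
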MaterialFunction /Game/Effects/Fort_Effects/BluePrints/WeakPointEffect/BlendMorphs
graph {nodes [
  n1 = MaterialFunctionCall() {materialFunction=/Engine/Functions/Engine_MaterialFunctions02/WorldPositionOffset/StaticMeshMorphTargets}
  n2 = Multiply(A=FunctionInput() {inputType=FunctionInput_Scalar}) {constB=2}
  n3 = Clamp(None=@n2)
  n4 = Subtract(A=@n2)
  FunctionOutput(None=LinearInterpolate(A=Multiply(A=@n1.Morph Target 1 WPO, B=@n3), B=@n1.Morph Target 2 WPO, Alpha=Clamp(None=@n4)))
 ]}
